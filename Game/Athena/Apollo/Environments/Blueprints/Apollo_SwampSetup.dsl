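; Apollo_SwampSetup is an actor that defines the swamp region for water on the Apollo map, split out of Apollo_WaterSetup after the Flood was removed.
; SetupSwamp runs from both the construction script and BeginPlay.
; It packs swamp X, Y and radius (in alpha) into MPC_Water's "SwampColor" vector parameter, despite that name.
; It also passes location and radius to every FortWaterBodyActor in the world.
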(event EventBeginPlay
  ;[1 1 1] Swamp setup moved from Apollo_WaterSetup into seperate actor for now since the Flood has gone - (SF)
  (CallFunction|SetupSwamp))

(fn SetupSwamp
  (bind _output_get (LocalVariables|SetSwampLocationAndRadius (Utilities|Struct|MakeLinearColor (.x (Variables|Default|GetSwampLocation)) (.y (Variables|Default|GetSwampLocation)) 0.0 (Variables|Default|GetSwampRadius))))
  (Rendering|Material|SetVectorParameterValue "/Water/Materials/MPC/MPC_Water.MPC_Water" "SwampColor" _output_get :!class /Script/Engine.KismetMaterialLibrary)
  (bind _outactors (Actor|GetAllActorsOfClass "/Script/FortniteGame.FortWaterBodyActor"))
  (for _array_element _outactors
    (WaterBody|SetSwampLocationandRadius _array_element (Variables|Default|GetSwampLocation) (Variables|Default|GetSwampRadius))))

(fn ConstructionScript
  (CallFunction|SetupSwamp))
