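; Apollo map setup actor that applies the season's water level, 0 to 7, to the world.
; On BeginPlay the server picks a debug or calendar level in debug/no-MCP runs, otherwise only logs event flags, then removes underwater trees and adjusts foundations.
; The CurrentWaterLevel RepNotify drives ocean flood height, terrain material parameters, swamp, minimap textures and ambient audio tags.
; An editor-only flood event updates foundations and trees for offline navmesh generation.
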(event EventBeginPlay
  ;[1 1 1] EVENT -> BeginPlay
  (if (Variables|Default|GetWaterSetupShouldRunBeginPlayLogic)
    (CallFunction|PerformEventFlagCheck)
    (if (Networking|HasAuthority)
      (if (and (Variables|Default|GetDebug) (Editor|InEditororPIE))
        (LevelStreaming|LoadLevelInstance(byName) (Utilities|String|Append "/Game/Maps/Apollo_Nav_Gameplay_WaterLevel_" (Utilities|String|ToString(Integer) (Variables|Default|GetDebugWaterLevel))))
        (bind _output_get (Variables|Default|SetCurrentWaterLevel (Variables|Default|GetDebugWaterLevel)))
        (Fortnite|FortShippingLog self (Utilities|String|Append "BeginPlay: Set w/ Notify Debug Water Level " (Utilities|String|ToString(Integer) _output_get)))
        (elif (Fortnite|IsRunningNoMCP)
          (bind _currentwaterlevelactive (CallFunction|CheckCalendarWaterLevel))
          (LevelStreaming|LoadLevelInstance(byName) (Utilities|String|Append "/Game/Maps/Apollo_Nav_Gameplay_WaterLevel_" (Utilities|String|ToString(Integer) _currentwaterlevelactive)))
          (bind _output_get_2 (Variables|Default|SetCurrentWaterLevel _currentwaterlevelactive))
          (Fortnite|FortShippingLog self (Utilities|String|Append "BeginPlay: Set w/ Notify Current Water Level " (Utilities|String|ToString(Integer) _output_get_2)))
          (else
            (MCPContext|InitialCalendarSyncCompleted
              (:OnSyncComplete
                ;;[1 1 1] If the setup actors starts up, and the calendar has been
                ;; synced, we better have a water level set. If not then we
                ;; default to water level 7 and load all the levels. We do not
                ;; support events turning on during the match.
                (bind _returnvalue (CallFunction|IsCalendarEventSet))
                (if _returnvalue
                  (MCPContext|WatchCalendarEvents (Variables|Default|GetEventsToWatch)
                    (:OnEventActive
                      (Utilities|FlowControl|DoOnce false
                        (:Completed
                          (CallFunction|PerformEventFlagCheck)))))
                  (else
                    _)))))))
      (Networking|ForceNetUpdate)
      (Variables|Default|SetWaterHeightSet true)
      (bind _outactors (Actor|GetAllActorsOfClass "/Game/Building/ActorBlueprints/Prop/Parent_Tree_Apollo.Parent_Tree_Apollo_C"))
      (for _array_element _outactors
        (Class|ParentTreeApollo|RemoveUnderwaterTrees _array_element))
      (CallFunction|AdjustWaterFoundations))))

(event WaterFlood|EventOnSetOceanFloodHeight
  (params (Step :type Integer))
  ;[1 1 1] Native Event (used offline in Editor for navmesh generation)
  (if (and (Variables|Default|GetDebug) (Editor|InEditororPIE))
    (bind _output_get (Variables|Default|SetCurrentWaterLevel Step))
    (Fortnite|FortShippingLog self (Utilities|String|Append "OnSetOceanFloorHeight: Current Water Level " (Utilities|String|ToString(Integer) Step)))
    (CallFunction|SetWaterFloodLevel _output_get)
    (Fortnite|FortShippingLog self (Utilities|String|Append "OnSetOceanFloorHeight: After SetWaterFloodLevel Current Water Level " (Utilities|String|ToString(Integer) _output_get)))
    (CallFunction|AdjustWaterFoundations)
    ;;[1 1 1] Foundations are adjusted, now udpate submerged trees to affect navmesh or not
    (bind _outactors (Actor|GetAllActorsOfClass "/Game/Building/ActorBlueprints/Prop/Parent_Tree_Apollo.Parent_Tree_Apollo_C"))
    (for _array_element _outactors
      (Class|ParentTreeApollo|OnWaterLevelSetOffline _array_element))))

(event Custom|Debug_SetWaterHeight
  (bind _debugwaterlevel (Variables|Default|GetDebugWaterLevel))
  ;[1 1 1] Editor Button
  (if (and (Variables|Default|GetDebug) (Editor|InEditororPIE))
    (CallFunction|SetWaterFloodLevel _debugwaterlevel)
    (Fortnite|FortShippingLog self (Utilities|String|Append "Debug_SetWaterHeight: After SetWaterFloodLevel Debug Water Level " (Utilities|String|ToString(Integer) _debugwaterlevel)))
    (else
      (bind _currentwaterlevelactive (CallFunction|CheckCalendarWaterLevel))
      (CallFunction|SetWaterFloodLevel _currentwaterlevelactive)
      (Fortnite|FortShippingLog self (Utilities|String|Append "Debug_SetWaterHeight: After SetWaterFloodLevel Current Water Level " (Utilities|String|ToString(Integer) 0))))))

(event Custom|SetWaterLevel
  (params (WaterLevel :type Integer))
  (bind _foundationmanager (Variables|Default|GetFoundationManager))
  ;[1 1 1] Ingame Console command
  (Networking|SwitchHasAuthority
    (:Authority
      (Variables|Default|SetCurrentWaterLevel WaterLevel)
      (Fortnite|FortShippingLog self (Utilities|String|Append "SetWaterLevel: Set w/ Notify Debug Water Level " (Utilities|String|ToString(Integer) WaterLevel)))
      (Utilities|IsValid _foundationmanager
        (:Is%20Valid
          (Class|ApolloWaterSetupFoundations|LoadCurrentWaterAudio _foundationmanager))))))

(event Custom|UpdateMinimapData
  (params (MapTexture :type Texture%202D%20Soft%20Object%20Reference) (DiscoveryTexture :type Texture%202D%20Soft%20Object%20Reference))
  (Class|FortAsyncActionGameStateReady|GameStateReadyAsync
    (:OnReady
      (CallFunction|SetMinimapData MapTexture DiscoveryTexture))))

(fn AdjustWaterFoundations
  (bind _foundationmanager (Variables|Default|GetFoundationManager))
  (Utilities|IsValid _foundationmanager
    (:Is%20Valid
      (Class|ApolloWaterSetupFoundations|SetWaterFoundations _foundationmanager (Variables|Default|GetCurrentWaterLevel)))
    (:Is%20Not%20Valid
      (Fortnite|FortShippingLog self "FoundationManager Not Valid"))))

(fn CheckCalendarWaterLevel
  (bind _biseventactive (CalendarEvents|GetCalendarEventInformation "WL6"))
  (if _biseventactive
    (LocalVariables|SetWaterLevel 6)
    (else
      (bind _biseventactive_2 (CalendarEvents|GetCalendarEventInformation "WL5"))
      (if _biseventactive_2
        (LocalVariables|SetWaterLevel 5)
        (else
          (bind _biseventactive_3 (CalendarEvents|GetCalendarEventInformation "WL4"))
          (if _biseventactive_3
            (LocalVariables|SetWaterLevel 4)
            (else
              (bind _biseventactive_4 (CalendarEvents|GetCalendarEventInformation "WL3"))
              (if _biseventactive_4
                (LocalVariables|SetWaterLevel 3)
                (else
                  (bind _biseventactive_5 (CalendarEvents|GetCalendarEventInformation "WL2"))
                  (if _biseventactive_5
                    (LocalVariables|SetWaterLevel 2)
                    (else
                      (bind _biseventactive_6 (CalendarEvents|GetCalendarEventInformation "WL1"))
                      (if _biseventactive_6
                        (LocalVariables|SetWaterLevel 1)
                        (else
                          (bind _biseventactive_7 (CalendarEvents|GetCalendarEventInformation "WL0"))
                          (if _biseventactive_7
                            (LocalVariables|SetWaterLevel 0)
                            (else
                              (LocalVariables|SetWaterLevel 7)))))))))))))))
  (return (LocalVariables|GetWaterLevel)))

(fn IsCalendarEventSet
  (bind _output_get (LocalVariables|SetEventManager (Fortnite|GetSeasonalEventManager)))
  (Utilities|IsValid _output_get
    (:Is%20Valid
      (for _array_element (Variables|Default|GetEventsToWatch)
        (if (Quests|IsEventFlagActive (LocalVariables|GetEventManager) _array_element)
          (return true))))
    (:Is%20Not%20Valid
      _))
  (Fortnite|FortShippingLog self (Variables|Default|GetWarningText) true)
  (return false))

(fn OnRep_CurrentWaterLevel
  (bind _currentwaterlevel (Variables|Default|GetCurrentWaterLevel))
  (CallFunction|SetWaterFloodLevel _currentwaterlevel)
  (Fortnite|FortShippingLog self (Utilities|String|Append "OnRepCurrentWaterLevel: Current Water Level " (Utilities|String|ToString(Integer) _currentwaterlevel)))
  (CallFunction|UpdateAmbientAudioWaterLevel))

(fn PerformEventFlagCheck
  (bind _returnvalue (Fortnite|GetSeasonalEventManager))
  (Utilities|IsValid _returnvalue
    (:Is%20Valid
      (Fortnite|FortShippingLog self "[WaterSetup] Found Seasonal Event Manager" true)
      (for _array_element (Variables|Default|GetEventsToWatch)
        (Fortnite|FortShippingLog self (Utilities|String|Append _array_element " is Active [" (Utilities|String|ToString(Boolean) (Quests|IsEventFlagActive _returnvalue _array_element)) "]") true)))))

(fn SetupSwamp
  (params (WaterLevel :type Integer))
  (bind _returnvalue (Math|Integer|Clamp(Integer) WaterLevel 0 1))
  (bind _swampradius (Variables|Default|GetSwampRadius))
  (bind _returnvalue_2 (- 1 _returnvalue))
  (bind _returnvalue_3 (* _returnvalue_2 _swampradius))
  (bind _swamplocation (Variables|Default|GetSwampLocation))
  (bind _output_get (LocalVariables|SetSwampLocationAndRadius (Utilities|Struct|MakeLinearColor (.x _swamplocation) (.y _swamplocation) 0.0 _returnvalue_3)))
  (Rendering|Material|SetVectorParameterValue "/Water/Materials/MPC/MPC_Water.MPC_Water" "SwampColor" _output_get :!class /Script/Engine.KismetMaterialLibrary)
  (bind _outactors (Actor|GetAllActorsOfClass "/Script/FortniteGame.FortWaterBodyActor"))
  (for _array_element _outactors
    (WaterBody|SetSwampLocationandRadius _array_element _swamplocation _returnvalue_3)))

(fn SetWaterFloodLevel
  (params (WaterLevel :type Integer))
  (bind _returnvalue (WorldSubsystems|GetWaterSubsystem))
  (Fortnite|FortShippingLog self (Utilities|String|Append "Set Water Flood Level: Passed In Water Level" (Utilities|String|ToString(Integer) WaterLevel)))
  (Water|SetOceanFloodHeight _returnvalue (* WaterLevel (Variables|Default|GetGridZSize)))
  (Rendering|Material|SetScalarParameterValue "/Game/Athena/Apollo/Environments/Landscape/MPC/Apollo_MPC_Terrain.Apollo_MPC_Terrain" "WaterLevel" (Water|GetOceanTotalHeight _returnvalue) :!class /Script/Engine.KismetMaterialLibrary) ; Will need to move this to correct MPC asset
  (Rendering|Material|SetScalarParameterValue "/Game/Athena/Apollo/Environments/Landscape/MPC/Apollo_MPC_Terrain.Apollo_MPC_Terrain" "WaterLevelGrid" (Math|Conversions|ToFloat(Integer) WaterLevel) :!class /Script/Engine.KismetMaterialLibrary) ; Will need to move this to correct MPC asset
  (CallFunction|SetupSwamp WaterLevel)
  (CallFunction|UpdateMinimapData (Utilities|Array|Get(acopy) (Variables|Default|GetMinimapTextures) WaterLevel) (Utilities|Array|Get(acopy) (Variables|Default|GetMinimapDiscoveryMasks) WaterLevel)))

(fn UpdateAmbientAudioWaterLevel
  (bind _returnvalue (WorldSubsystems|GetAmbientAudioSubsystem))
  (bind _water_level_tag (Variables|Default|GetWater%20Level%20Tag))
  (for _array_element _water_level_tag
    (AmbientAudio|RemoveGameplayTag _returnvalue _array_element))
  (AmbientAudio|AddGameplayTag _returnvalue (Utilities|Array|Get(acopy) _water_level_tag (Variables|Default|GetCurrentWaterLevel))))

(fn ConstructionScript
  (bind _debugwaterlevel (Variables|Default|GetDebugWaterLevel))
  (bind _returnvalue (Math|Conversions|ToFloat(Integer) _debugwaterlevel))
  (bind _returnvalue_2 (Editor|InEditororPIE))
  (bind _floodlevel (Variables|Default|GetFloodLevel))
  (bind _returnvalue_3 (ScalableFloat|GetValueatLevel _floodlevel))
  (bind _debug (Variables|Default|GetDebug))
  (bind _returnvalue_4 (and _debug _returnvalue_2))
  (bind _returnvalue_5 (select _returnvalue_4 _returnvalue _returnvalue_3))
  (CallFunction|SetupSwamp 0)
  (Rendering|Material|SetScalarParameterValue "/Game/Athena/Apollo/Environments/Landscape/MPC/Apollo_MPC_Terrain.Apollo_MPC_Terrain" "WaterLevelGrid" _returnvalue_5 :!class /Script/Engine.KismetMaterialLibrary) ; This needs to be a global parameter
  (Fortnite|FortShippingLog self (Utilities|String|Append "Construction Water Level " (Utilities|String|ToString(Float) _returnvalue_5))))
