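; Blueprint for Apollo water-level setup, with helpers that place building foundations and swap streamed levels as the ocean height changes.
; LoadCurrentWaterAudio swaps the ocean audio sublevel for the cached water level and, at level zero, sets ShouldLoadRuins, whose OnRep streams in Ruins water.
; Helpers derive grid-snapped spline transforms and water heights, then load foundations via editor preview or runtime dynamic foundations.
; SetWaterFoundations is empty; debug events step an arrow along the PawnShop spline.
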(event Custom|LoadCurrentWaterAudio
  (bind _currentwaterlevelcached (Variables|Default|GetCurrentWaterLevelCached))
  (bind _levelname (CallFunction|GetAudioSplineLevel _currentwaterlevelcached))
  ;[1 1 1] Load Audio Spline Level - Ruins Overlay level
  (for _index (range 8)
    (Variables|/Script/Engine.LevelStreaming|SetbShouldBeLoaded false (Game|GetStreamingLevel (CallFunction|GetAudioSplineLevel _index))) ; This seems to be required?
    (Game|UnloadStreamLevel(byName) (Game|GetWorldAssetPackageFName (Game|GetStreamingLevel (CallFunction|GetAudioSplineLevel _index)))))
  (Game|UnloadStreamLevel(byName) "Apollo_POI_TheRuins_Water")
  (Variables|/Script/Engine.LevelStreaming|SetbShouldBeLoaded true (Game|GetStreamingLevel _levelname))
  (Game|LoadStreamLevel(byName) _levelname true)
  ;[1 1 1] Load ruins water level
  (if (== (Variables|Default|GetCurrentWaterLevelCached) 0)
    (Variables|Default|SetShouldLoadRuins true)))

(event Custom|LoadRuins
  ;[1 1 1] Load ruins water level
  (Game|LoadStreamLevel(byName) "Apollo_POI_TheRuins_Water" true))

(event Custom|SetWaterFoundations
  (params (WaterLevel :type Integer)))

(event Custom|DebugSpline%20+1
  (bind _spline_pawnshop (Variables|Default|GetSpline_PawnShop))
  ;[1 0 0] debug
  (Variables|Default|SetDebugSplinePointIndex (Math|Integer|Clamp(Integer) (+ (Variables|Default|GetDebugSplinePointIndex) 1) 0 (- (Spline|GetNumberOfSplinePoints _spline_pawnshop) 1)))
  (CallFunction|DebugSplineTransform _spline_pawnshop))

(event Custom|DebugSpline%20-1
  (bind _spline_pawnshop (Variables|Default|GetSpline_PawnShop))
  ;[1 0 0] debug
  (Variables|Default|SetDebugSplinePointIndex (Math|Integer|Clamp(Integer) (- (Variables|Default|GetDebugSplinePointIndex) 1) 0 (- (Spline|GetNumberOfSplinePoints _spline_pawnshop) 1)))
  (CallFunction|DebugSplineTransform _spline_pawnshop))

(event Custom|DebugSpline%20Reset
  ;[1 0 0] debug
  (Variables|Default|SetDebugSplinePointIndex -1)
  (CallFunction|DebugSplineTransform (Variables|Default|GetSpline_PawnShop)))

(fn Debug_StreamingExtents
  (params (Foundation :type Building%20Foundation%20Object%20Reference) (Location :type Vector))
  (bind (_origin _boxextent) (Collision|GetActorBounds Foundation false false))
  (bind _returnvalue (AddComponent|NODE_AddBoxComponent-4 :RelativeTransform_Location Location :RelativeTransform_Rotation "0, 0, 0" :RelativeTransform_Scale "1.000000,1.000000,1.000000"))
  (Components|Box|SetBoxExtent _returnvalue (+ (select (Variables|Default|GetShowMobileStreamingExtents) (Variables|Default|GetMobileStreaming_In) (Variables|Default|GetStreamingExtents_In)) _boxextent) true)
  (bind _returnvalue_2 (AddComponent|NODE_AddBoxComponent-5 :RelativeTransform_Location Location :RelativeTransform_Rotation "0, 0, 0" :RelativeTransform_Scale "1.000000,1.000000,1.000000"))
  (Components|Box|SetBoxExtent _returnvalue_2 (+ (select (Variables|Default|GetShowMobileStreamingExtents) (Variables|Default|GetMobileStreaming_Out) (Variables|Default|GetStreamingExtents_Out)) _boxextent) true))

(fn DebugSplineTransform
  (params (SplineToTest :type Spline%20Component%20Object%20Reference))
  (bind _debugsplinepointindex (Variables|Default|GetDebugSplinePointIndex))
  (bind _location (CallFunction|GetGridLocationonSpline SplineToTest 0.0 true _debugsplinepointindex))
  (Transformation|SetWorldTransform (Variables|Default|GetDebugArrow) (Math|Transform|MakeTransform _location (CallFunction|GetGridRotationonSpline 0.0 0.0 SplineToTest _debugsplinepointindex) "300.000000,300.000000,300.000000") false false))

(fn GetAudioSplineLevel
  (params (WaterLevel :type Integer))
  (return (Utilities|String|StringToName (Utilities|String|Append "Apollo_Audio_Ocean_" (select (> WaterLevel 9) (Utilities|String|ToString(Integer) WaterLevel) (Utilities|String|Append "0" (Utilities|String|ToString(Integer) WaterLevel)))))))

(fn GetGridLocationOnSpline
  (params (Spline :type Spline%20Component%20Object%20Reference) (Progress :type Float%20%28double-precision%29) (UseSplinePoint :type Boolean) (SplinePoint :type Integer))
  (return (Fortnite|SnaptoFortBuildingGrid (Math|Vector|SelectVector (Spline|GetLocationatSplinePoint Spline SplinePoint "World") (Spline|GetLocationatDistanceAlongSpline Spline (* (Spline|GetSplineLength Spline) Progress) "World") UseSplinePoint))))

(fn GetGridRotationOnSpline
  (params (Progress :type Float%20%28double-precision%29) (ForwardVectorYawRotation :type Float%20%28double-precision%29) (Spline :type Spline%20Component%20Object%20Reference) (SplinePoint :type Integer))
  (bind _returnvalue (== SplinePoint (- (Spline|GetNumberOfSplinePoints Spline) 1)))
  (bind _returnvalue_2 (- SplinePoint 1))
  (bind _returnvalue_3 (Math|Integer|SelectInt _returnvalue_2 SplinePoint _returnvalue))
  (bind _returnvalue_4 (Math|Conversions|ToFloat(Integer) _returnvalue_3))
  (bind _returnvalue_5 (Spline|GetSplineLength Spline))
  (bind _returnvalue_6 (Math|Conversions|ToFloat(Integer) (- (Spline|GetNumberOfSplinePoints Spline) 1)))
  (bind _returnvalue_7 (/ _returnvalue_5 _returnvalue_6))
  (bind _returnvalue_8 (Math|Rotator|MakeRotator 0.0 0.0 ForwardVectorYawRotation))
  (bind _returnvalue_9 (* _returnvalue_7 _returnvalue_4))
  (bind _returnvalue_10 (Spline|GetDirectionatDistanceAlongSpline Spline _returnvalue_9 "World"))
  (bind _returnvalue_11 (Math|Vector|RotateVector _returnvalue_10 _returnvalue_8))
  (bind _returnvalue_12 (Math|Conversions|RotationFromXVector _returnvalue_11))
  (bind _returnvalue_13 (Spline|GetNumberOfSplinePoints Spline))
  (bind _returnvalue_14 (- _returnvalue_13 1))
  (bind _yaw (.yaw _returnvalue_12))
  (Development|PrintString (Utilities|String|ToString(Float) _yaw))
  (return (Math|Rotator|MakeRotator 0.0 0.0 (Math|Float|SnapToGrid(Float) _yaw 9e+01)) _yaw))

(fn GetLocationForLevel
  (params (Foundation :type Building%20Foundation%20Object%20Reference) (GroundLevel :type Integer) (UseWaterZ :type Boolean))
  (return (Math|Vector|SelectVector (CallFunction|SetLocationZToWaterHeight (Transformation|GetActorLocation Foundation) GroundLevel) (Transformation|GetActorLocation Foundation) UseWaterZ) (Transformation|GetActorRotation Foundation)))

(fn GetSplinePointFromWaterLevel
  (params (Spline :type Spline%20Component%20Object%20Reference) (StartingWaterLevel :type Integer) (CurrentWaterLevel :type Integer) (FinishedPathBias :type Integer))
  (return (Math|Integer|Clamp(Integer) (- (- (Variables|Default|GetMaxWaterLevel) CurrentWaterLevel) (- (Variables|Default|GetMaxWaterLevel) (Math|Integer|SelectInt (Math|Integer|Clamp(Integer) StartingWaterLevel 0 (Variables|Default|GetMaxWaterLevel)) (Variables|Default|GetMaxWaterLevel) (<= (Math|Integer|Clamp(Integer) StartingWaterLevel 0 (Variables|Default|GetMaxWaterLevel)) (Variables|Default|GetMaxWaterLevel))))) 0 (- (Spline|GetNumberOfSplinePoints Spline) 1)) (> (- (- (- (Variables|Default|GetMaxWaterLevel) CurrentWaterLevel) (- (Variables|Default|GetMaxWaterLevel) (Math|Integer|SelectInt (Math|Integer|Clamp(Integer) StartingWaterLevel 0 (Variables|Default|GetMaxWaterLevel)) (Variables|Default|GetMaxWaterLevel) (<= (Math|Integer|Clamp(Integer) StartingWaterLevel 0 (Variables|Default|GetMaxWaterLevel)) (Variables|Default|GetMaxWaterLevel))))) FinishedPathBias) (- (Spline|GetNumberOfSplinePoints Spline) 1))))

(fn GetWaterHeightFromGrid
  (params (WaterLevel :type Integer))
  (return (+ (* (Math|Conversions|ToFloat(Integer) WaterLevel) 384.0) (Variables|Default|GetMinOceanHeight))))

(fn LoadFoundation
  (params (Foundation :type Building%20Foundation%20Object%20Reference))
  (if (Editor|InEditororPIE)
    ;[1 1 1] In Editor world we use the editor preview flow
    (bind _returnvalue (Class|BuildingFoundation|EditorOnlyLoadPreviewLevel Foundation true))
    (if _returnvalue
      _
      (else
        _))
    (else
      _))
  ;[1 1 1] In Game world we use runtime dynamic foundation flow
  (Class|BuildingFoundation|SetDynamicFoundationEnabled Foundation true))

(fn MoveAndLoadFoundation
  (params (Foundation :type Building%20Foundation%20Object%20Reference) (NewLocation :type Vector) (NewRotation :type Rotator))
  (if (Editor|InEditororPIE)
    ;[1 1 1] In Editor world we use the editor preview flow
    (bind _returnvalue (Class|BuildingFoundation|EditorOnlyAddTransformtoPreviewLevel Foundation (Math|Transform|MakeTransform NewLocation NewRotation)))
    (if _returnvalue
      (Class|BuildingFoundation|EditorOnlyLoadPreviewLevel Foundation true)
      (else
        _)))
  ;[1 1 1] In Game world we use runtime dynamic foundation flow
  (Class|BuildingFoundation|SetDynamicFoundationTransform Foundation (Math|Transform|MakeTransform NewLocation NewRotation))
  (Class|BuildingFoundation|SetDynamicFoundationEnabled Foundation true))

(fn OnRep_ShouldLoadRuins
  (if (Variables|Default|GetShouldLoadRuins)
    (CallFunction|LoadRuins)))

(fn SetLocationZToWaterHeight
  (params (LocationIn :type Vector) (GroundLevel :type Integer))
  (return (Math|Vector|MakeVector (.x LocationIn) (.y LocationIn) (+ (CallFunction|GetWaterHeightfromGrid (Math|Integer|Max(Integer) GroundLevel (Variables|Default|GetCurrentWaterLevelCached))) 61.0))))

(fn ConstructionScript)
