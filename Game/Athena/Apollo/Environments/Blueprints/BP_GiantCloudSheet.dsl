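; BP_GiantCloudSheet is an environment Blueprint for a giant cloud sheet, with events to toggle its shadow proxy.
; EventHideShadowProxy and EventShowShadowProxy set the visibility of the SM_FX_1-Unit_Octagon mesh component off and on, without propagating to children.
; The construction script is empty, and the file holds no other logic.
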(event EventHideShadowProxy
  (Rendering|SetVisibility (Variables|Default|GetSM_FX_1-Unit_Octagon) false false))

(event EventShowShadowProxy
  (Rendering|SetVisibility (Variables|Default|GetSM_FX_1-Unit_Octagon) true false))

(fn ConstructionScript)
