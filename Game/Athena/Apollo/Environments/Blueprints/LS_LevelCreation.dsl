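; Editor tool Blueprint that divides an Apollo environment into a square grid of streaming sublevels.
; Numbered events populate per-cell level lists, bucket actors by their top-level parent's grid cell (currently gathering only HLOD fake-geo actors), then move them into sublevels and clear references for saving.
; Helper events create or align sublevels, spawn foundations linked to each level, add text labels, and place fake geometry as an HLOD-generation workaround for imposter-only cells.
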(event Custom|1%29%20PopulateLevelLists
  ;[1 1 1] Clear arrays
  (Utilities|Array|Clear (Variables|Default|GetLevelOffsetList))
  (Utilities|Array|Clear (Variables|Default|GetLevelPathNames))
  (Utilities|Array|Clear (Variables|Default|GetStreamingLevelList))
  ;[1 1 1] Create Grid and information to associated index for each grid square index
  (bind (_levelname _gridcellcenterlocation _x _y _currentgridcell) (Utilities|CalculateGrid (Variables|Default|GetGridAmount))
    (:PerGridExecution
      (Utilities|Array|Add (Variables|Default|GetLevelPathNames) _levelname)
      (Utilities|Array|Add (Variables|Default|GetLevelOffsetList) _gridcellcenterlocation)
      (Utilities|Array|Add (Variables|Default|GetStreamingLevelList) (Game|GetStreamingLevel 0 (Utilities|String|StringToName _levelname)))
      ;[1 1 1] Debugging for lists
      (if (Variables|Debugging|GetShowDebug)
        (Development|PrintString 0 (Utilities|String|Append _levelname "is Grid " "X: " (Utilities|String|ToString(Integer) _x) ", Y: " (Utilities|String|ToString(Integer) _y) ", Grid Number = " (Utilities|String|ToString(Integer) _currentgridcell)))
        (Rendering|Debug|DrawDebugBox :Center _gridcellcenterlocation :Extent (/ (Math|Vector|MakeVector (.x (Variables|Default|GetCellSize)) (.y (Variables|Default|GetCellSize)) (.y (Variables|Default|GetCellSize))) 2.0) :LineColor "(R=0.000000,G=1.000000,B=0.296237,A=1.000000)" :Duration (Variables|Debugging|GetBoxDebugDuration) :Thickness 0.0 :DepthPriority "World")))))

(event Custom|AlignSublevelsToGrid
  (bind _streaminglevellist (Variables|Default|GetStreamingLevelList))
  ;[1 1 1] Align existing sublevels to grid
  (bind (_levelname _gridcellcenterlocation _x _y _currentgridcell) (Utilities|CalculateGrid (Variables|Default|GetGridAmount))
    (:PerGridExecution
      (Development|PrintString 0 (Utilities|GetDisplayName (Utilities|Array|Get(acopy) _streaminglevellist _currentgridcell)))
      (Variables|Default|SetNewVar_0 (.location (Variables|/Script/Engine.LevelStreaming|GetLevelTransform (Utilities|Array|Get(acopy) _streaminglevellist _currentgridcell))))
      (bind _output_get (Variables|/Script/Engine.LevelStreaming|SetLevelTransform (Math|Transform|MakeTransform (Utilities|Array|Get(aref) (Variables|Default|GetLevelOffsetList) _currentgridcell)) (Utilities|Array|Get(acopy) _streaminglevellist _currentgridcell)))
      (Development|PrintString 0 (Utilities|String|Append "Old Location:" (Utilities|String|ToString(Vector) (Variables|Default|GetNewVar_0)) ",  New Location: " (Utilities|String|ToString(Vector) (.location _output_get)))))
    (:Completed)))

(event Custom|CreateGridFoundations
  ;[1 1 1] Spawn Foundations
  (for _array_element (Variables|Default|GetFoundations)
    (Actor|DestroyActor _array_element))
  (Utilities|Array|Clear (Variables|Default|GetFoundations))
  (for (_array_element_2 _array_index) (Variables|Default|GetLevelPathNames)
    (bind _returnvalue (Game|SpawnActorfromClass 0 (Variables|Default|GetFoundation%20Class) (Math|Conversions|ToTransform(Vector) (Utilities|Array|Get(aref) (Variables|Default|GetLevelOffsetList) _array_index)) "Undefined" "SelectDefaultAtRuntime" 0 0 0 0))
    (Class|BuildingFoundation|EditorOnlyClearAdditionalWorlds _returnvalue)
    (Class|BuildingFoundation|EditorOnlyAddAdditionalWorld _returnvalue _array_element_2)
    (Utilities|Array|Add (Variables|Default|GetFoundations) _returnvalue)))

(event Custom|2%29%20PrepareSplitData
  (bind _cellsize (Variables|Default|GetCellSize))
  (bind _parentresult (Variables|Default|GetParentResult))
  (bind _returnvalue (Transformation|GetActorLocation _parentresult))
  (bind _grid_cell (CallFunction|LocateGridCell _returnvalue _cellsize))
  (Utilities|Array|Clear (Variables|Default|GetLevelSplitData))
  ;[1 1 1] Set array size
  (Utilities|Array|SetArrayElem (Variables|Default|GetLevelSplitData) (- (* (Variables|Default|GetGridAmount) (Variables|Default|GetGridAmount)) 1) (Utilities|Struct|MakeLevelActorData (Utilities|Array|MakeArray 0)) true)
  (bind _outactors (Actor|GetAllActorsOfClass 0 "/Game/Athena/Apollo/Environments/Blueprints/LS_HLODGenHack_FakeGeo.LS_HLODGenHack_FakeGeo_C"))
  (bind _array_element (Utilities|Array|ForEachLoopwithBreak _outactors)
    (:LoopBody
      ;[1 1 1] Filter Actors
      (Utilities|SkipIgnoredClasses (Utilities|GetClass _array_element)
        (:ValidClass
          ;;[1 1 1] Bad hack to skip referencing the levels themselves which cant be filtered by class - just in case
          ;[1 1 1] Use parent level for attached assets
          (Utilities|GetTopLevelParent (Variables|Default|GetParentResult) _array_element
            (:Out
              (Utilities|Array|Add (Utilities|Struct|BreakLevelActorData (Utilities|Array|Get(aref) (Variables|Default|GetLevelSplitData) (- _grid_cell 1))) _array_element)
              ;[1 1 1] Debugging
              (if (Variables|Debugging|GetShowDebug)
                (Development|PrintString 0 (Utilities|String|Append (Utilities|GetDisplayName (Variables|Default|GetParentResult)) " of class type: " (Utilities|GetClassDisplayName (Utilities|GetClass _array_element)) " is in: " (Utilities|String|ToString(Integer) _grid_cell) "But belongs to level: " (Utilities|String|ToString(Name) (Game|GetWorldAssetPackageFName (Utilities|GetStreamingLevelfromObject _array_element)))))))))))
    (:Completed
      ;[1 1 1] Debugging
      (if (Variables|Debugging|GetShowDebug)
        (for (_array_element_2 _array_index) (Variables|Default|GetLevelSplitData)
          (for _array_element_3 (Utilities|Struct|BreakLevelActorData _array_element_2)
            (Development|PrintString 0 (Utilities|String|Append "Struct: " (Utilities|String|ToString(Integer) _array_index) " Contains: " (Utilities|GetDisplayName _array_element_3)))))))))

(event Custom|SpawnGridLabels
  ;[1 1 1] Spawn Grid labels for debugging/working
  (for (_array_element _array_index) (Variables|Default|GetLevelPathNames)
    (bind _returnvalue (EditorScripting|LevelUtility|SpawnActorfromClass "/Game/Athena/Apollo/Environments/Blueprints/LS_LevelGrid_Labels.LS_LevelGrid_Labels_C" (+ (Utilities|Array|Get(aref) (Variables|Default|GetLevelOffsetList) _array_index) (Math|Vector|MakeVector 0.0 0.0 5e+03)) :!class /Script/EditorScriptingUtilities.EditorLevelLibrary))
    (Rendering|Components|TextRender|SetText (Variables|/Game/Athena/Apollo/Environments/Blueprints/LS_LevelGrid_Labels.LS_LevelGrid_Labels_C|GetTextRender _returnvalue) (Utilities|Text|ToText(String) (Utilities|Array|Get(aref) (Utilities|String|ParseIntoArray _array_element "/") (Utilities|Array|LastIndex (Utilities|String|ParseIntoArray _array_element "/")))))))

(event Custom|ClearArrays
  (bind _levelsplitdata (Variables|Default|GetLevelSplitData))
  ;[1 1 1] Clear out references in Arrays to be able to save
  (for _array_element _levelsplitdata
    (Utilities|Array|Clear (Utilities|Struct|BreakLevelActorData _array_element))
    (Variables|Default|SetParentResult self))
  (Utilities|Array|Clear _levelsplitdata)
  (Utilities|Array|Clear (Variables|Default|GetLevelOffsetList))
  (Utilities|Array|Clear (Variables|Default|GetLevelPathNames))
  (Utilities|Array|Clear (Variables|Default|GetStreamingLevelList)))

(event Custom|3%29%20MoveActorsToSubLevels
  (for (_array_element _array_index) (Variables|Default|GetLevelSplitData)
    (bind _returnvalue (EditorScripting|LevelCreation|MoveActorstoLevel (Utilities|Struct|BreakLevelActorData _array_element) (Utilities|Array|Get(acopy) (Variables|Default|GetStreamingLevelList) _array_index) false))
    ;[1 1 1] Debug
    (if (Variables|Debugging|GetShowDebug)
      (Development|PrintString 0 (Utilities|String|Append (Utilities|String|ToString(Integer) _returnvalue) " of " (Utilities|String|ToString(Integer) (Utilities|Array|Length (Utilities|Struct|BreakLevelActorData _array_element))) " Moved to " ""))))
  (CallFunction|ClearArrays))

(event Custom|SpawnHLODGenPlanes
  ;[1 0 0] Used as a temporary hack to work around Grid levels not building HLODS when containing only Imposters
  (for (_ _array_index) (Variables|Default|GetLevelPathNames)
    (EditorScripting|LevelUtility|SpawnActorfromClass "/Game/Athena/Apollo/Environments/Blueprints/LS_HLODGenHack_FakeGeo.LS_HLODGenHack_FakeGeo_C" (+ (Utilities|Array|Get(aref) (Variables|Default|GetLevelOffsetList) _array_index) (Math|Vector|MakeVector 0.0 0.0 -8e+03)) :!class /Script/EditorScriptingUtilities.EditorLevelLibrary)))

(event Custom|CreateSublevels
  ;[1 0.4153 0] Create Sublevels - Don't use by accident
  (bind (_levelname _gridcellcenterlocation) (Utilities|CalculateGrid (Variables|Default|GetGridAmount))
    (:PerGridExecution
      (Development|PrintString 0 _levelname)
      (if (Variables|Default|GetEnableCreateSublevels)
        (bind _returnvalue (EditorScripting|LevelCreation|CreateNewStreamingLevel (Variables|Default|GetStreamingMethod) _levelname))
        (if (Variables|Default|GetOffsetLevel)
          (Variables|/Script/Engine.LevelStreaming|SetLevelTransform (Math|Transform|MakeTransform _gridcellcenterlocation) _returnvalue))))))

(fn LocateGridCell
  (params (Location :type Vector) (CellSize :type Vector))
  (return (+ (Utilities|2dTo1dIndex (Math|Vector2D|MakeVector2D (Math|Conversions|ToFloat(Integer) (Math|Float|Floor (.x (/ (+ (* Location "-1.000000,1.000000,1.000000") (/ (Math|Float|Absolute(Float) (Variables|Default|GetLandscapeWidth)) 2.0)) CellSize)))) (Math|Conversions|ToFloat(Integer) (Math|Float|Floor (.y (/ (+ (* Location "-1.000000,1.000000,1.000000") (/ (Math|Float|Absolute(Float) (Variables|Default|GetLandscapeWidth)) 2.0)) CellSize))))) (Variables|Default|GetGridAmount)) 1)))

(fn ParseStringAndRemove
  (params (string :type String))
  (if (Utilities|String|EndsWith string "_C")
    (LocalVariables|SetParsedString (Utilities|String|LeftChop string 2))
    (else
      (LocalVariables|SetParsedString string)))
  (return (LocalVariables|GetParsedString)))

(fn SearchStringArray
  (params (Filenames :type Array%20of%20Strings) (SubString :type String))
  (bind _array_element (Utilities|Array|ForEachLoopwithBreak Filenames)
    (:LoopBody
      (if (Utilities|String|Contains _array_element SubString)
        (LocalVariables|SetNewLocalVar_0 true)))
    (:Completed
      _))
  (if (Utilities|String|Contains (Variables|Default|GetPersistantLevelName) SubString)
    (LocalVariables|SetNewLocalVar_0 true))
  (return (LocalVariables|GetNewLocalVar_0)))

(fn ConstructionScript
  (Variables|Default|SetCellSize (/ (* (Math|Vector|VectorGetAbs (Variables|Default|GetStartLocation)) 2) (Math|Vector|MakeVector (Math|Conversions|ToFloat(Integer) (Variables|Default|GetGridAmount)) (Math|Conversions|ToFloat(Integer) (Variables|Default|GetGridAmount)) 1.0))))
